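# Material function that returns the intersection of two distance fields, exposed to the material library under Distance Fields.
# It takes two scalar inputs, Distance Field 1 and Distance Field 2.
# A single Max node combines them and feeds the function output, so the result is the larger of the two distances.
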MaterialFunction /Engine/Functions/DistanceFields/Combiners/DistanceField_Intersection
graph {
 properties {bExposeToLibrary=true, description=Returns the intersection of two distance field functions., libraryCategoriesText [Distance Fields]}
 nodes [FunctionOutput(None=Max(
    A=FunctionInput() {inputName=Distance Field 1, inputType=FunctionInput_Scalar}
    B=FunctionInput() {inputName=Distance Field 2, inputType=FunctionInput_Scalar, sortPriority=1}
   ))]
}
