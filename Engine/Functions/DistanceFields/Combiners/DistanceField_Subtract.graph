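# DistanceField_Subtract is a library material function that carves one signed distance field shape out of another.
# It negates Distance Field 2 by multiplying it by -1, then takes the Max of that and Distance Field 1.
# This is the standard CSG difference operation for distance fields, with both inputs and the result as scalars.
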MaterialFunction /Engine/Functions/DistanceFields/Combiners/DistanceField_Subtract
graph {
 properties {bExposeToLibrary=true, description=Subtracts Distance Field 2 from Distance Field 1, libraryCategoriesText [Distance Fields]}
 nodes [
  n1 = FunctionInput() {
   description=The second shape. This is the shape that will be subtracted.
   inputName=Distance Field 2
   inputType=FunctionInput_Scalar
   sortPriority=1
  }
  FunctionOutput(None=Max(A=FunctionInput() {description=The fist shape, inputName=Distance Field 1, inputType=FunctionInput_Scalar}, B=Multiply(A=@n1) {constB=-1}))
 ]
}
