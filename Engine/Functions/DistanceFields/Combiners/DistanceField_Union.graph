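# Material function in the engine's Distance Fields library that returns the union of two distance field shapes.
# It takes two scalar inputs, Distance Field 1 and 2, and outputs their per-pixel maximum through a single Max node.
# Using Max rather than Min for union implies fields whose values increase toward the inside of a shape.
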MaterialFunction /Engine/Functions/DistanceFields/Combiners/DistanceField_Union
graph {
 properties {bExposeToLibrary=true, description=Returns the Union of two distance field shapes., libraryCategoriesText [Distance Fields]}
 nodes [FunctionOutput(None=Max(
    A=FunctionInput() {inputName=Distance Field 1, inputType=FunctionInput_Scalar}
    B=FunctionInput() {inputName=Distance Field 2, inputType=FunctionInput_Scalar, sortPriority=1}
   ))]
}
